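# Master material for the vending machine monitor screen, blending between a weapon icon on a gradient backing and an animated emote face via bIsLlama.
# The emote layers two swappable textures, blink, laugh bob and optional prop over radial flair, clouds and rainbow.
# Slide and noise-driven block distortion shift the shared UVs, with distortion dropped at low quality.
# A malfunction switch adds scanlines before a border mask and emissive adjustment.
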Material /Game/Athena/Items/Gameplay/VendingMachine/Emotes/M_VendingMachineMonitor_Emote_Master
graph {
 properties {bUsedWithNanite=true, pixelDepthOffsetMode=PDOM_Legacy, usedShadingModels=Default Lit}
 nodes [
  Constant() {r=1}
  Constant() {r=0.15}
  n1 = Time()
  n2 = ScalarParameter() {defaultValue=22, parameterName=Block Speed A}
  n3 = Panner(Coordinate=TextureCoordinate() {uTiling=0.2}, Time=Ceil(None=Multiply(A=@n1, B=@n2))) {bFractionalPart=true, speedX=0.7, speedY=-0.3}
  n4 = Multiply(
   A=LinearInterpolate(Alpha=TextureSample(UVs=@n3) {samplerType=SAMPLERTYPE_LinearColor, texture=/Game/Global/Textures/Default/Noise/T_Atlas_DigitalNoise_01}.B) {constA=-0.1, constB=0.1}
   B=Add(A=ScalarParameter() {parameterName=Distort}, B=ScalarParameter() {parameterName=DistortionAdd})
  )
  n5 = ComponentMask(None=TextureCoordinate() {vTiling=2.2}) {g=true}
  n6 = Multiply(A=Time()) {constB=0.2}
  n7 = Floor(None=Multiply(A=Frac(None=@n6)) {constB=1.1})
  n8 = Multiply(A=Multiply(A=Floor(None=@n5), B=@n7) {constA=0.3}, B=@n7)
  n9 = AppendVector(A=Clamp(None=Multiply(B=@n8) {constA=-0.03}) {minDefault=-1}, B=Constant())
  n10 = Add(A=QualitySwitch(Default=Add(A=@n4, B=@n9), Low=Constant()), B=Multiply(A=ScalarParameter() {parameterName=Slide}, B=Constant2Vector() {r=1}))
  n11 = Add(A=TextureCoordinate(), B=@n10)
  n12 = AppendVector(A=Constant(), B=Multiply(A=Time(), B=ScalarParameter() {defaultValue=0.5, parameterName=ScanlineTimeFactor}))
  n13 = Add(A=Sine(None=Add(A=@n11, B=@n12)) {period=0.025}) {constB=-0.5}
  n14 = Multiply(A=Saturate(None=ComponentMask(None=@n13) {g=true}), B=ScalarParameter() {defaultValue=0.5, parameterName=ScanlineIntensity})
  n15 = Multiply(A=TextureCoordinate(), B=Constant2Vector() {g=2, r=1})
  n16 = Add(
   A=MaterialFunctionCall(UVs=Add(A=@n15, B=@n10), Texture Scale=ScalarParameter() {defaultValue=0.85, parameterName=WeaponIconScale}) {materialFunction=&ScaleUVsByCenter}.UVs
   B=AppendVector(A=Constant(), B=ScalarParameter() {defaultValue=-0.55, parameterName=WeaponIcon-Y-Offset})
  )
  n17 = Add(A=@n16) {constB=-0.5}
  n18 = Saturate(None=Round(None=Abs(None=@n17)))
  n19 = ComponentMask(None=@n18) {r=true}
  n20 = ComponentMask(None=@n18) {g=true}
  n21 = TextureSampleParameter2D(UVs=@n16) {
   parameterName=Weapon-Icon
   samplerSource=SSM_Clamp_WorldGroupSettings
   texture=/Game/Athena/UI/Challenges/Art/ChallengeTypeIcons/UI_ChallengeIcon_Pistol
  }
  n22 = Multiply(A=OneMinus(None=Max(A=@n19, B=@n20)), B=@n21.A)
  n23 = ScalarParameter() {defaultValue=1, parameterName=WeaponIconIntensity}
  n24 = VectorParameter() {parameterName=BG_2_RGBA, defaultValue=<rgba 0 0.0801 0.5833 1>}
  n25 = Add(A=Multiply(A=TextureCoordinate(), B=Constant2Vector() {g=2, r=1}), B=@n10)
  n26 = Saturate(None=MaterialFunctionCall(
    UVs=@n25
    CenterPosition=Constant2Vector() {g=1, r=0.5}
    Radius=ScalarParameter() {defaultValue=1.125, parameterName=Radius}
    Density=ScalarParameter() {defaultValue=1.5, parameterName=Density}
   ) {materialFunction=&RadialGradientExponential}.RadialGradientExponential)
  n27 = Subtract(A=@n26, B=ScalarParameter() {defaultValue=0.2857, parameterName=WeaponIconBacking Size}) {constB=0.5}
  n28 = Multiply(A=Saturate(None=Multiply(A=@n27) {constB=64})) {constB=0.5}
  n29 = VectorParameter() {parameterName=BG_2_RGBA, defaultValue=<rgba 0 0.0801 0.5833 1>}
  n30 = LinearInterpolate(A=@n29.RGB, B=VectorParameter() {parameterName=BG_1_RGBA, defaultValue=<rgba 0 0.4018 0.5833 1>}.RGB, Alpha=@n26)
  n31 = TextureCoordinate()
  n32 = Constant2Vector() {g=2, r=1}
  n33 = MaterialFunctionCall(Vector or UVs=Add(A=Multiply(A=@n31, B=@n32), B=Constant2Vector() {g=-1.125, r=-0.5})) {materialFunction=&VectorToRadialValue}
  n34 = Panner(Coordinate=@n33.Vector Converted to Angle) {speedX=0.125, speedY=0.25}
  n35 = MaterialFunctionCall(Float2=Saturate(None=Sine(None=@n34) {period=0.2})) {materialFunction=&BreakOutFloat2Components}
  n36 = Sine(None=@n35.R)
  n37 = Power(Base=Add(A=Multiply(A=@n35.R) {constB=0.25}, B=Saturate(None=@n36)), Exp=Saturate(None=Power(Base=@n33.Linear Distance) {constExponent=4})) {constExponent=0.25}
  n38 = VectorParameter() {parameterName=Cloud and Rainbow, defaultValue=<rgba 0.225 -0.65 0.65 0>}
  n39 = LinearInterpolate(
   B=VectorParameter() {parameterName=BG_1_RGBA, defaultValue=<rgba 0 0.4018 0.5833 1>}.RGB
   Alpha=Multiply(A=Multiply(A=@n37, B=@n38.A), B=ScalarParameter() {defaultValue=0.125, parameterName=Flair-Intensity})
  )
  n40 = Constant() {r=-1.125}
  n41 = TextureObject() {texture=/Game/Athena/Items/Gameplay/VendingMachine/Emotes/EmoteTextures/T_VendingMachine-Clouds}
  n42 = TextureSample(UVs=Add(A=@n25, B=AppendVector(A=@n38.R, B=@n40)), Tex=@n41) {samplerSource=SSM_Clamp_WorldGroupSettings, texture=&T_VendingMachine-Clouds}
  n43 = Power(Base=@n42.R, Exp=ScalarParameter() {defaultValue=4, parameterName=Cloud-1-Power})
  n44 = Add(A=@n30, B=LinearInterpolate(B=LinearInterpolate(A=@n29.RGB, Alpha=@n43) {constA=1}, Alpha=@n42.R))
  n45 = Constant() {r=-0.75}
  n46 = TextureSample(UVs=Add(A=@n25, B=AppendVector(A=@n38.G, B=@n45)), Tex=@n41) {samplerSource=SSM_Clamp_WorldGroupSettings, texture=&T_VendingMachine-Clouds}
  n47 = ScalarParameter() {defaultValue=4, parameterName=Cloud-2-Power}
  n48 = LinearInterpolate(B=LinearInterpolate(A=@n29.RGB, Alpha=Power(Base=@n46.G, Exp=@n47)) {constA=1}, Alpha=@n46.G)
  n49 = Constant()
  n50 = TextureSample(UVs=Add(A=@n25, B=AppendVector(A=@n38.B, B=@n49)), Tex=@n41) {samplerSource=SSM_Clamp_WorldGroupSettings, texture=&T_VendingMachine-Clouds}
  n51 = LinearInterpolate(A=@n29.RGB, Alpha=Power(Base=@n50.G, Exp=ScalarParameter() {defaultValue=4, parameterName=Cloud-2-Power})) {constA=1}
  n52 = Constant3Vector() {constant=<rgba 0.9647 0.2051 0.4233 1>}
  n53 = Constant3Vector() {constant=<rgba 0.3278 0.7529 0.4397 1>}
  n54 = Constant2Vector() {g=1, r=1}
  n55 = Constant()
  n56 = LinearInterpolate(B=ScalarParameter() {defaultValue=-0.05, parameterName=RainbowYOffset}, Alpha=@n38.A) {constA=-0.25}
  n57 = TextureSample(UVs=Add(A=Multiply(A=@n25, B=@n54), B=AppendVector(A=@n55, B=@n56))) {samplerSource=SSM_Clamp_WorldGroupSettings, texture=/Game/Athena/Items/Gameplay/VendingMachine/Emotes/EmoteTextures/T_VendingMachine-Rainbow}
  n58 = ScalarParameter() {defaultValue=1, parameterName=RainBoxIntensity}
  n59 = LinearInterpolate(
   A=Add(A=Add(A=@n44, B=@n48), B=LinearInterpolate(A=@n48, B=@n51, Alpha=@n50.G))
   B=LinearInterpolate(
    A=LinearInterpolate(A=@n52, B=@n53, Alpha=@n57.R)
    B=Constant3Vector() {constant=<rgba 0.8469 0.9131 0.1981 1>}
    Alpha=DotProduct(A=@n57.R, B=@n57.R)
   )
   Alpha=Multiply(A=Multiply(A=@n57.B, B=@n58), B=@n38.A)
  )
  n60 = Multiply(A=TextureCoordinate(), B=Constant2Vector() {g=2, r=1})
  n61 = ScalarParameter() {defaultValue=1.25, parameterName=IconScale}
  n62 = ScalarParameter() {defaultValue=-0.5, parameterName=Icon-Y-Offset}
  n63 = Multiply(A=Time(), B=ScalarParameter() {defaultValue=2, parameterName=LaughTime})
  n64 = LinearInterpolate(B=LinearInterpolate(Alpha=Sine(None=@n63)) {constA=0.0125, constB=-0.0125}, Alpha=ScalarParameter() {parameterName=bUseLaugh})
  n65 = Add(
   A=Add(
    A=MaterialFunctionCall(UVs=@n60, Texture Scale=@n61) {materialFunction=&ScaleUVsByCenter}.UVs
    B=@n10
   )
   B=AppendVector(A=Constant(), B=Add(A=@n62, B=@n64))
  )
  n66 = TextureSampleParameter2D(UVs=@n65) {
   parameterName=Emote_A
   samplerSource=SSM_Clamp_WorldGroupSettings
   texture=/Game/Athena/Items/Gameplay/VendingMachine/Emotes/EmoteTextures/T_VendingMachine-E-happy
  }
  n67 = ScalarParameter() {parameterName=EmoteSwap}
  n68 = TextureSampleParameter2D(UVs=@n65) {
   parameterName=Emote_B
   samplerSource=SSM_Clamp_WorldGroupSettings
   texture=/Game/Athena/Items/Gameplay/VendingMachine/Emotes/EmoteTextures/T_VendingMachine-E-Excited
  }
  n69 = LinearInterpolate(A=@n66.RGB, B=@n68.RGB, Alpha=@n67)
  n70 = Time() {bOverride_Period=true, period=4}
  n71 = ScalarParameter() {defaultValue=8, parameterName=TimeFactor}
  n72 = Saturate(None=LinearInterpolate(Alpha=Multiply(A=@n70, B=@n71)) {constA=1, constB=0})
  n73 = TextureSample(UVs=@n65) {samplerSource=SSM_Clamp_WorldGroupSettings, texture=&T_VendingMachine-Clouds}
  n74 = LinearInterpolate(
   A=@n69
   B=LinearInterpolate(A=@n69, B=Constant3Vector() {constant=<rgba 0.6654 0.1946 0.7835 1>}, Alpha=Multiply(A=@n72, B=@n73.B))
   Alpha=ScalarParameter() {defaultValue=1, parameterName=bUseBlink}
  )
  n75 = TextureSampleParameter2D(UVs=@n65) {
   parameterName=Prop
   samplerSource=SSM_Clamp_WorldGroupSettings
   texture=&T_VendingMachine-E-happy
  }
  n76 = LinearInterpolate(A=@n74, B=@n75.RGB, Alpha=@n75.A)
  n77 = LinearInterpolate(A=@n66.A, B=@n68.A, Alpha=@n67)
  n78 = LinearInterpolate(A=@n77, Alpha=@n75.A)
  n79 = LinearInterpolate(
   A=LinearInterpolate(A=LinearInterpolate(A=@n30, B=@n24.RGB, Alpha=@n28), Alpha=Multiply(A=@n22, B=@n23))
   B=LinearInterpolate(
    A=Add(A=@n39, B=@n59)
    B=StaticSwitchParameter(True=@n76, False=@n74) {parameterName=bHasProp}
    Alpha=StaticSwitchParameter(True=@n78, False=@n77) {parameterName=bHasProp}
   )
   Alpha=ScalarParameter() {parameterName=bIsLlama}
  )
  n80 = StaticSwitchParameter(True=Add(A=Multiply(A=@n14, B=@n79), B=@n79), False=@n79) {parameterName=bIsMalfunction}
  n81 = TextureCoordinate()
  n82 = Subtract(A=Abs(None=Add(A=@n81) {constB=-0.5}), B=Constant2Vector() {g=0.48, r=0.465})
  n83 = Constant2Vector() {g=256, r=128}
  n84 = MaterialFunctionCall(Vector 2=Saturate(None=Multiply(A=@n82, B=@n83))) {materialFunction=/Engine/Functions/Engine_MaterialFunctions02/Utility/VectorLength}
  n85 = OneMinus(None=Saturate(None=Add(B=@n84.V2 Length)))
  MaterialFunctionCall(Emissive=Multiply(A=Multiply(A=@n80, B=@n85), B=ScalarParameter() {defaultValue=0.5, parameterName=Emissive Intensity}) {constB=0.125}) {materialFunction=&MF_FX_EmissiveAdjustments}
  ScalarParameter() {defaultValue=1, parameterName=bUseRainbow}
  SphereMask(
   A=ObjectPositionWS()
   B=CameraPositionWS()
   Radius=ScalarParameter() {defaultValue=600}
   Hardness=ScalarParameter() {defaultValue=0.25, parameterName=Param_1}
  )
 ]
}
